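# Conversation database for the Deadfire NPC, defining its greeting dialogue and the services it leads to.
# The entry point opens a speech node with a default line and prioritized contextual alternatives for returning players, first-ever meetings and a Flapjack Wrangler outfit.
# The speech branches through link nodes into the shared NPC library quests and basic services conversations.
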ConversationDatabase /NPCLibrary/NPCs/Deadfire/Deadfire/NPCLibrary_Conversation_Deadfire
entry_points [{tag=Conversation.Entry.Deadfire, destinations [#1]}]
nodes [
 ConversationEntryPointNode_7 {
  id=#1
  class=ConversationEntryPointNode
  outputs [#2]
  properties {entryTag {tagName=Conversation.Entry.Deadfire}}
 }
 Speech_C_4 {
  id=#2
  class=Speech_C
  outputs [#3, #4]
  properties {
   nodeName=""
   generalConfig {
    defaultMessage="Make it quick, partner."
    contextualMessages [
     {
      message="Well well. Look who decided to saddle up again. <b>Good to see you.</>"
      contextRequirements [{requirement {class=FortControllerRequirement_CheckTimeSinceLastPlayedMatch}}]
      priority {curve {rowName=Tandem.Speech.Priority.WhileSincePlayed}}
     }
     {
      message="Ain't seen you before, cowpoke."
      contextRequirements [{requirement {class=FortControllerRequirement_IsFirstEverConversationWithNPC}}]
      priority {curve {rowName=Tandem.Speech.Priority.FirstEverInteraction}}
     }
     {
      message=I've got my eye on you. Don't try nothin' fishy.
      contextRequirements [{requirement {
         class=FortControllerRequirement_HasCID
         properties {allowedCIds [/BRCosmetics/Athena/Items/Cosmetics/Characters/CID_962_Athena_Commando_M_FlapjackWrangler]}
        }}]
      priority {curve {rowName=Tandem.Speech.Priority.OutfitReaction}}
     }
    ]
   }
  }
 }
 ConversationLinkNode_0 {id=#4, class=ConversationLinkNode, properties {remoteEntryTag {tagName=Conversation.Entry.BasicServices}}}
 ConversationLinkNode_1 {id=#3, class=ConversationLinkNode, properties {remoteEntryTag {tagName=Conversation.Entry.NPCLibrary.Quests}}}
]
